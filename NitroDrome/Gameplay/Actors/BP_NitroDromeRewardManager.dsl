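; Reward manager for the NitroDrome objective, running a timer based on participating pawns' presence.
; On completion the authority tosses loot from jumbotron sockets, schedules a one-time boss spawn broadcast and listens for storm phases to reset the objective.
; It broadcasts tagged events through the playspace event router and updates a minimap icon actor, while clients show progress via custom primitive data.
; ScalableFloat hotfixes gate the objective and reset, and cheat events support testing.
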(event Custom|CheatNitroDromeFakeStormHold
  ;[0.3574 0 1] Allow Event Reset on Storm Change.
  ; Storm Holding & Shrinking events are spammy while those phases are active, so Do Once on Storm Holding, reset gate on Storm Shrink
  (if (not (Variables|Default|GetObjectiveCanBeCompleted))
    (Utilities|FlowControl|DoOnce false
      (:Completed
        (Fortnite|FortShippingLog self "Setting FirstTimeThroughOnStormShrink as false and resetting objective event state")
        (Variables|GamePhase|SetFirstTimeThroughOnStormShrink false)
        (CallFunction|ResetNitroDromeObjectiveEventState)))))

(event Custom|CheckGamePhaseStep
  (params (Game%20Phase%20Step :type EAthenaGamePhaseStep%20Enum))
  ;[0.3574 0 1] Allow Event Reset on Storm Change.
  ; Storm Holding & Shrinking events are spammy while those phases are active, so Do Once on Storm Holding, reset gate on Storm Shrink
  (if (Utilities|Enum|Equal(Enum) Game%20Phase%20Step (Variables|GamePhase|GetStormHoldingMarkingStep))
    (Fortnite|FortShippingLog self (Utilities|String|Append "Detected Holding Storm Phase. ObjectiveCanBeCompleted: " (Utilities|String|ToString(Boolean) (Variables|Default|GetObjectiveCanBeCompleted))))
    (if (Variables|Default|GetObjectiveCanBeCompleted)
      _
      (else
        _))
    (elif (and (Utilities|Enum|Equal(Enum) Game%20Phase%20Step (Variables|GamePhase|GetStormShrinkingStep)) (not (Variables|GamePhase|GetFirstTimeThroughOnStormShrink)))
      (Variables|GamePhase|SetFirstTimeThroughOnStormShrink true)
      (Fortnite|FortShippingLog self "Storm Shrinking Step detected. FirstTimeThroughOnStormShrink is false and  FirstTimeThroughOnStormShrink is now true")))
  (Utilities|FlowControl|DoOnce false
    (:Completed
      (Fortnite|FortShippingLog self "Setting FirstTimeThroughOnStormShrink as false and resetting objective event state")
      (Variables|GamePhase|SetFirstTimeThroughOnStormShrink false)
      (CallFunction|ResetNitroDromeObjectiveEventState))))

(event Custom|CheatNitroDromeSpamEvents
  (params (Enabled :type Boolean))
  (bind _testcheattimerhandle (Variables|Default|GetTestCheatTimerHandle))
  ;[1 1 1] Cheat - fires off event on loop for audio etc. testing
  (if Enabled
    (bind _returnvalue (Utilities|Time|SetTimerbyEvent (delegate Custom|CheatNitroDromeSpamEvent) 5.0 true))
    (Variables|Default|SetTestCheatTimerHandle _returnvalue)
    (elif (Utilities|Time|IsValidTimerHandle _testcheattimerhandle)
      (Utilities|Time|ClearandInvalidateTimerbyHandle _testcheattimerhandle))))

(event Custom|CheatNitroDromeSpamEvent
  ;[1 1 1] Cheat - fires off event on loop for audio etc. testing
  (CallFunction|BroadcastReactableEvent "(GameplayTags=((TagName=\"NitroDrome.Event.LootEventReset\")))"))

(event Custom|CheatNitroDromeFakeStormShrink
  ;[0.3574 0 1] Allow Event Reset on Storm Change.
  ; Storm Holding & Shrinking events are spammy while those phases are active, so Do Once on Storm Holding, reset gate on Storm Shrink
  (if (not (Variables|GamePhase|GetFirstTimeThroughOnStormShrink))
    (Variables|GamePhase|SetFirstTimeThroughOnStormShrink true)
    (Fortnite|FortShippingLog self "Storm Shrinking Step detected. FirstTimeThroughOnStormShrink is false and  FirstTimeThroughOnStormShrink is now true")
    (Utilities|FlowControl|DoOnce false
      (:Completed
        (Fortnite|FortShippingLog self "Setting FirstTimeThroughOnStormShrink as false and resetting objective event state")
        (Variables|GamePhase|SetFirstTimeThroughOnStormShrink false)
        (CallFunction|ResetNitroDromeObjectiveEventState)))))

(event Custom|StartListeningToStormEvents
  ;[0.3574 0 1] Allow Event Reset on Storm Change.
  ; Storm Holding & Shrinking events are spammy while those phases are active, so Do Once on Storm Holding, reset gate on Storm Shrink
  (if (ScalableFloat|GetValueasBool (Variables|Hotfix|GetSF_ResetEnabeld))
    (bind _returnvalue (GameplayEventRouter|GetEventRouter "/Script/FortniteGame.FortGameState" self))
    (bind (_asynctaskproxy _payload) (GameplayEventRouter|StartListeningToStatefulEvent _returnvalue "/Script/FortniteGame.FortGamePhaseStepUpdatedEvent")
      (:then
        (Variables|GamePhase|SetGamePhaseStepUpdatedAsyncTask _asynctaskproxy)
        (Fortnite|FortShippingLog self "Started listening to storm events!"))
      (:OnEventReceived
        (CallFunction|CheckGamePhaseStep (Utilities|Struct|BreakFortGamePhaseStepUpdatedEvent _payload))))))

(event Custom|CheatNitroDromeAffectTimer
  (params (Time :type Float%20%28double-precision%29))
  ;[1 1 1] Cheat - any time input is subtracted from the timer
  (bind _returnvalue (GameplayEventRouter|GetEventRouter "/NitroDrome/Gameplay/Actors/BP_NitroDromePlayspace.BP_NitroDromePlayspace_C" self))
  (GameplayEventRouter|BroadcastEvent _returnvalue (Utilities|Struct|MakeOnDromeTimeAffectingActionHappened (* Time -1.0) self "(GameplayTags=((TagName=\"NitroDrome.Event.Checkpoint.Used\")))") 0))

(event EventOnWorldReady
  (Networking|SwitchHasAuthority
    (:Authority
      ;[0 0.01859 1] Hotfix enabled/disabled
      (if (ScalableFloat|GetValueasBool (Variables|Hotfix|GetSF_ObjectiveEnabled))
        (Variables|Default|SetObjectiveCanBeCompleted true)
        (MinimapIcon|SpawnMinimapIcon)
        (bind _returnvalue (GameplayEventRouter|GetEventRouter "/NitroDrome/Gameplay/Actors/BP_NitroDromePlayspace.BP_NitroDromePlayspace_C" self))
        (Utilities|FlowControl|Sequence
          (:then_0
            (bind (_asynctaskproxy _payload) (GameplayEventRouter|StartListeningToEvent _returnvalue "/Script/FortniteGame.ActorTrackerActorEnteredContext")
              (:OnEventReceived
                (bind _payload_enteredactor (Utilities|Struct|BreakActorTrackerActorEnteredContext _payload))
                (CallFunction|OnNitroDromePresenceChanged true _payload_enteredactor))))
          (:then_1
            (bind (_asynctaskproxy_2 _payload_2) (GameplayEventRouter|StartListeningToEvent _returnvalue "/Script/FortniteGame.ActorTrackerActorExitContext")
              (:OnEventReceived
                (bind _payload_exitingactor (Utilities|Struct|BreakActorTrackerActorExitContext _payload_2))
                (CallFunction|OnNitroDromePresenceChanged false _payload_exitingactor)))))
        (else
          (Actor|DestroyActor))))))

(event EventOnTimerStarted
  (Development|PrintString "REWARD TIMER STARTED")
  (Networking|SwitchHasAuthority
    (:Remote
      (Audio|SpawnSoundatLocation "/NitroDrome/Assets/Audio/Checkpoints/NitroDrome_Checkpoint_Timer_Start_Cue.NitroDrome_Checkpoint_Timer_Start_Cue" (Transformation|GetWorldLocation (Variables|/NitroDrome/Assets/Blueprints/BP_NitroDrome_Center_A.BP_NitroDrome_Center_A_C|GetSpeakers (Variables|Default|GetJumbotron))))
      (CallFunction|UpdateRewardContainerColorForCurrentStatus)
      (bind _returnvalue (Utilities|Time|SetTimerbyEvent (EventDispatchers|CreateEvent :!event Custom|UpdateTimerProgress) 0.5 true true))
      (Variables|Default|SetProgressTimerUpdateTimerHandle _returnvalue))))

(event Custom|CheatForceNitroDromeLootSpawn
  (Variables|Default|SetObjectiveCanBeCompleted true)
  (Networking|SwitchHasAuthority
    (:Authority
      (CallFunction|SpawnRewardItems)
      (CallFunction|BroadcastReactableEvent "(GameplayTags=((TagName=\"NitroDrome.Event.LootRewardSpawned\")))")
      (CallFunction|SpawnLootRewardVFX)
      (Variables|Default|SetObjectiveCanBeCompleted false)
      (CallFunction|BroadcastNItroDromeBossSpawn)
      (CallFunction|StartListeningtoStormEvents)
      (MinimapIcon|SetClaimedMinimapIcon))
    (:Remote
      (CallFunction|UpdateRewardContainerProgressVisual 1.0)
      (CallFunction|UpdateRewardContainerColorForCurrentStatus)
      (Utilities|Time|ClearandInvalidateTimerbyHandle (Variables|Default|GetProgressTimerUpdateTimerHandle)))))

(event EventOnTimerFinished
  (Development|PrintString "REWARD TIMER Finished")
  (Networking|SwitchHasAuthority
    (:Authority
      (CallFunction|SpawnRewardItems)
      (CallFunction|BroadcastReactableEvent "(GameplayTags=((TagName=\"NitroDrome.Event.LootRewardSpawned\")))")
      (CallFunction|SpawnLootRewardVFX)
      (Variables|Default|SetObjectiveCanBeCompleted false)
      (CallFunction|BroadcastNItroDromeBossSpawn)
      (CallFunction|StartListeningtoStormEvents)
      (MinimapIcon|SetClaimedMinimapIcon))
    (:Remote
      (CallFunction|UpdateRewardContainerProgressVisual 1.0)
      (CallFunction|UpdateRewardContainerColorForCurrentStatus)
      (Utilities|Time|ClearandInvalidateTimerbyHandle (Variables|Default|GetProgressTimerUpdateTimerHandle)))))

(event EventOnTimerAffected
  (params (AffectingData_TimeDelta :type Float%20%28single-precision%29) (AffectingData_Author :type Actor%20Object%20Reference) (AffectingData_EventTags :type Gameplay%20Tag%20Container%20Structure))
  (Default|CallOnTimerAffectingActionOccurred AffectingData_TimeDelta AffectingData_Author AffectingData_EventTags)
  (Development|PrintString (Utilities|String|Append "Reward time affected by: " (Utilities|String|ToString(Float) AffectingData_TimeDelta))))

(event EventOnRewardApproaching
  (Networking|SwitchHasAuthority
    (:Authority
      (CallFunction|BroadcastReactableEvent "(GameplayTags=((TagName=\"NitroDrome.Event.Reward.Approaching\")))"))))

(event EventOnTimerCleared
  (Development|PrintString "REWARD TIMER CLEARED")
  (Networking|SwitchHasAuthority
    (:Remote
      (CallFunction|UpdateRewardContainerProgressVisual 0.0)
      (CallFunction|UpdateRewardContainerColorForCurrentStatus)
      (Utilities|Time|ClearandInvalidateTimerbyHandle (Variables|Default|GetProgressTimerUpdateTimerHandle)))))

(event Custom|UpdateTimerProgress
  (CallFunction|UpdateRewardContainerProgressVisual (CallFunction|GetProgressPercentage)))

(event Custom|ResetNitroDromeObjectiveEventState
  ;[0 1 0.07173] Reset objective allowed & broadcast event to reset checkpoint visibility etc. Objective allowed bool gates UpdateRewardTimerBasedOnPresence()
  (Networking|SwitchHasAuthority
    (:Authority
      (if (not (Variables|Default|GetObjectiveCanBeCompleted))
        (Variables|Default|SetObjectiveCanBeCompleted true)
        (CallFunction|BroadcastReactableEvent "(GameplayTags=((TagName=\"NitroDrome.Event.LootEventReset\")))")
        (CallFunction|UpdateTimerBasedOnPresence)
        (MinimapIcon|SetUnclaimedMinimapIcon)))))

(event Custom|SpawnLootRewardVFX
  (bind _jumbotron (Variables|Default|GetJumbotron))
  (bind _returnvalue (Transformation|GetActorLocation _jumbotron))
  (bind _returnvalue_2 (+ _returnvalue "0.000000,0.000000,680.000000"))
  (Effects|Components|ParticleSystem|SpawnFXAtLocation (Variables|Default|GetLootRewardFXSystem%20Asset) _returnvalue_2)
  (SoundIndicator|ShowSoundIndicatorforDuration (Variables|Default|GetFortSoundIndicator) _returnvalue_2 "/Game/Sounds/Helios/POI/NitroDrome/LootRewards/NitroDrome_LootReward_Burst_Cue.NitroDrome_LootReward_Burst_Cue" 1.0 3.0 0))

(event Custom|BroadcastNItroDromeBossSpawn
  ;[1 0 0.5203] Broadcast Boss Spawn. Set bool so we don't do this on subsequent loot event completions
  (Networking|SwitchHasAuthority
    (:Authority
      (if (Utilities|Enum|Equal(Enum) (Variables|Default|GetBossStatus) "NewEnumerator0")
        (bind _returnvalue (Utilities|Time|SetTimerbyEvent (delegate Custom|Try%20Broadcast%20Boss%20Spawn) (ScalableFloat|GetValueatLevel (Variables|Hotfix|GetSF_BossBroadcastDelay))))
        (Variables|Default|SetBossBroadcastTimer _returnvalue)))))

(event Custom|Try%20Broadcast%20Boss%20Spawn
  ;[1 0 0.5203] Broadcast Boss Spawn. Set bool so we don't do this on subsequent loot event completions
  (CallFunction|BroadcastReactableEvent "(GameplayTags=((TagName=\"NitroDrome.Event.Boss.Spawn\")))")
  (Variables|Default|SetBossStatus "NewEnumerator1")
  (MinimapIcon|SetMinimapIcon (Variables|MapIcon|GetMinimapIconClaimed))
  (Utilities|Time|ClearandInvalidateTimerbyHandle (Variables|Default|GetBossBroadcastTimer))
  (GameplayEventRouter|StartListeningToEvent (GameplayEventRouter|GetGlobalEventRouter) "/Script/FortniteGame.FortVerbMessage_PawnDied"))

(event EventOnRewardStatusChanged
  (params (OldRewardStatus :type Nitro%20Drome%20Reward%20Status%20Data%20Structure%20%28by%20ref%29) (NewRewardStatus :type Nitro%20Drome%20Reward%20Status%20Data%20Structure%20%28by%20ref%29)))

(fn BroadcastReactableEvent
  (params (Event%20Tag%20Container :type Gameplay%20Tag%20Container%20Structure))
  (bind _returnvalue (GameplayEventRouter|GetEventRouter "/NitroDrome/Gameplay/Actors/BP_NitroDromePlayspace.BP_NitroDromePlayspace_C" self))
  (GameplayEventRouter|BroadcastEvent _returnvalue (Utilities|Struct|MakeOnNitroDromeReactableEventHappened Event%20Tag%20Container) 0))

(fn OnNitroDromePresenceChanged
  (params (bEntered :type Boolean) (Actor :type Actor%20Object%20Reference))
  (bind _asfort_player_pawn (Utilities|Casting|CastToFortPlayerPawn Actor))
  (if (GameplayTags|DoesContainerMatchTagQuery (GameplayTags|GetOwnedGameplayTags _asfort_player_pawn :!class /Script/GameplayTags.GameplayTagAssetInterface) (GameplayTags|MakeGameplayTagQuery "(TokenStreamVersion=0,TagDictionary=((TagName=\"Pawn.Athena.NonParticipant\")),QueryTokenStream=(0,1,3,1,0),UserDescription=\"Non-NPC Pawn\",AutoDescription=\" NONE( Pawn.Athena.NonParticipant )\")"))
    (bind _output_get (Variables|Default|SetNumParticipatingPawns (+ (Variables|Default|GetNumParticipatingPawns) (Math|Integer|SelectInt 1 -1 bEntered))))
    (CallFunction|UpdateTimerBasedOnPresence)
    (Fortnite|FortShippingLog self (Utilities|String|Append "Pawn " (Utilities|GetDisplayName (Variables|GetActor)) (Utilities|String|SelectString " entered " " exited " (Variables|GetbEntered)) "the Nitro Drome." "NumPawns = " (Utilities|String|ToString(Integer) _output_get)))))

(fn OnRep_BossStatus
  (Default|CallClientonBossStateChanged (Variables|Default|GetBossStatus)))

(fn ResetRewardManager
  ;[1 1 1] Check if we have any participating pawns on a reset, so we start the timer instead of forcing players to exit and come back in
  (if (> (Variables|Default|GetNumParticipatingPawns) 0)
    (Timer|StartTimer 0 0.0)
    (else
      (Timer|ClearTimer))))

(fn SetClaimedMinimapIcon
  (MinimapIcon|SetMinimapIcon (select (Utilities|Enum|Equal(Enum) (Variables|Default|GetBossStatus) "NewEnumerator0") (Variables|MapIcon|GetMinimapIconMedallionLocked) (Variables|MapIcon|GetMinimapIconClaimed))))

(fn SetMinimapIcon
  (params (NewIcon :type Texture%202D%20Object%20Reference))
  (bind _mapiconactor (Variables|MapIcon|GetMapIconActor))
  (Utilities|IsValid _mapiconactor
    (:Is%20Valid
      (Class|BPNitroDromeMapIconActor|SetMapIcon _mapiconactor NewIcon))))

(fn SetUnclaimedMinimapIcon
  (MinimapIcon|SetMinimapIcon (select (Utilities|Enum|Equal(Enum) (Variables|Default|GetBossStatus) "NewEnumerator0") (Variables|MapIcon|GetMinimapIconMedallionLocked) (Variables|MapIcon|GetMinimapIconUnclaimed))))

(fn SpawnRewardItems
  (bind _itempickupsspawned (Default|SpawnLootfromTableUsingItemEntrywithMultipleSpawnLocations (select (!= (CallFunction|GetLootTierforCurrentPlaylist) "") (CallFunction|GetLootTierforCurrentPlaylist) (Variables|Default|GetDefaultRewardLootTierGroup)) 0 -1 (Transformation|GetWorldLocation (Variables|Default|GetLootSpawnReferenceTransform)) (Transformation|GetForwardVector (Variables|Default|GetLootSpawnReferenceTransform)) 0 true true false 0 "Other" "Unset" 0 false (Variables|Default|GetJumbotron) (ScalableFloat|GetValueatLevel (Variables|Hotfix|GetSF_LootTossDistance))))
  (for _array_element _itempickupsspawned
    (Utilities|Casting|CastToFortWeaponItemDefinition (Class|FortItemFunctionLibrary|BreakFortItemEntry (Pickup|GetItemEntry _array_element 0)))
    (Fortnite|Weapon|CreateTossAmmoPickupforWeaponItemDefinitionatLocation 0 (Variables|Default|GetRewardAmmoSourceTag) (Transformation|GetActorLocation))
    (Ability|ExecuteGameplayCueWithParamsReplicated (Variables|Default|GetGameplay%20Cue%20Tag) (Ability|GameplayCue|MakeGameplayCueParameters :NormalizedMagnitude 0.0 :RawMagnitude 0.0 :Location "0,0,0" :Normal "0,0,0" :SourceObject _array_element :GameplayEffectLevel 1 :AbilityLevel 1 :bReplicateLocationWhenUsingMinimalRepProxy false))))

(fn SpawnLootFromTableUsingItemEntryWithMultipleSpawnLocations
  (params (TierGroupName :type Name) (WorldLevel :type Integer :default 1) (ForcedLootTier :type Integer :default -1) (Position :type Vector) (Direction :type Vector) (OverrideMaxStackCount :type Integer) (bToss :type Boolean) (bRandomRotation :type Boolean) (bBlockedFromAutoPickup :type Boolean) (PickupInstigatorHandle :type Integer) (SourceType :type EFortPickupSourceTypeFlag%20Enum :default "Other") (Source :type EFortPickupSpawnSource%20Enum :default "Unset") (OptionalOwnerPC :type Fort%20Player%20Controller%20Object%20Reference) (bPickupOnlyRelevantToOwner :type Boolean) (Jumbotron :type BP%20Nitro%20Drome%20Center%20A%20Object%20Reference) (TossDistance :type Float%20%28double-precision%29))
  (bind _jumbotronsocketindex (LocalVariables|GetJumbotronSocketIndex))
  (bind _jumbotronsocketarray (LocalVariables|GetJumbotronSocketArray))
  (bind _output (Utilities|Array|Get(aref) _jumbotronsocketarray _jumbotronsocketindex))
  (bind _staticmeshcomponent (Variables|/Script/FortniteGame.BuildingSMActor|GetStaticMeshComponent Jumbotron))
  (bind (_returnvalue_location _returnvalue_rotation) (Math|Transform|BreakTransform (Transformation|GetSocketTransform _staticmeshcomponent _output "RTS_World")))
  (bind _returnvalue (Math|Vector|GetForwardVector _returnvalue_rotation))
  (LocalVariables|SetJumbotronSocketIndex 0)
  (Utilities|Array|Clear (LocalVariables|GetPickupsSpawned))
  (bind (_outloottodrop _returnvalue_2) (Loot|PickLootDrops TierGroupName WorldLevel ForcedLootTier "(GameplayTags=)"))
  (if _returnvalue_2
    (for _array_element _outloottodrop
      (bind _returnvalue_3 (Inventory|SpawnPickupInWorldWithClassAndItemEntry 0 _array_element 0 _returnvalue_location _returnvalue OverrideMaxStackCount bToss bRandomRotation bBlockedFromAutoPickup SourceType Source OptionalOwnerPC bPickupOnlyRelevantToOwner))
      (Math|Integer|IncrementInt (LocalVariables|GetJumbotronSocketIndex)
        (:%20%20
          (if (>= (LocalVariables|GetJumbotronSocketIndex) (Utilities|Array|Length (LocalVariables|GetJumbotronSocketArray)))
            (LocalVariables|SetJumbotronSocketIndex 0)
            (else
              _))
          (Utilities|Array|AddUnique (LocalVariables|GetPickupsSpawned) _returnvalue_3)
          (Pickup|TossPickup _returnvalue_3 (+ _returnvalue_location (* _returnvalue TossDistance)) 0 0 true true "Tossed" "Unset"))))
    (return (LocalVariables|GetPickupsSpawned))
    (else
      (Development|PrintString 0 "NoLootPackageFound"))))

(fn SpawnMinimapIcon
  (Networking|SwitchHasAuthority
    (:Authority
      ;[0 0 0] Create the MapIconActor. This is how Heist Bases and God Challenges handled the icon to avoid relevancy issues.
      (bind _returnvalue (Game|SpawnActorfromClass "/NitroDrome/Gameplay/Actors/BP_NitroDrome_MapIconActor.BP_NitroDrome_MapIconActor_C" (Math|Transform|MakeTransform (Transformation|GetActorLocation) (Transformation|GetActorRotation)) "AlwaysSpawn" "MultiplyWithRoot" 0 0))
      (Variables|MapIcon|SetMapIconActor _returnvalue)
      (MinimapIcon|SetUnclaimedMinimapIcon))))

(fn UpdateRewardContainerColorforCurrentStatus
  (bind _rewardcontainer (Variables|Default|GetRewardContainer))
  (switch Utilities|FlowControl|Switch|SwitchonENitroDromeRewardStatus (CallFunction|GetCurrentRewardStatus)
    (:Inactive
      (Rendering|Material|SetCustomPrimitiveDataFloat _rewardcontainer 0 0.0))
    (:Active
      (Rendering|Material|SetCustomPrimitiveDataFloat _rewardcontainer 0 2.0))
    (:Finished
      (Rendering|Material|SetCustomPrimitiveDataFloat _rewardcontainer 0 3.0))))

(fn UpdateRewardContainerProgressVisual
  (params (FillPercentage :type Float%20%28double-precision%29))
  (Rendering|Material|SetCustomPrimitiveDataFloat (Variables|Default|GetRewardContainer) (Variables|Default|GetRewardProgressDataFloatIndex) FillPercentage))

(fn UpdateTimerBasedonPresence
  (if (> (Variables|Default|GetNumParticipatingPawns) 0)
    ;[1 1 1]
    (if (Variables|Default|GetObjectiveCanBeCompleted)
      (switch Utilities|FlowControl|Switch|SwitchonENitroDromeRewardStatus (CallFunction|GetCurrentRewardStatus)
        (:Inactive
          _)
        (:Paused
          (Timer|ResumeTimer))
        (:Finished
          _))
      (Timer|StartTimer 0 0.0)
      (CallFunction|BroadcastReactableEvent "(GameplayTags=((TagName=\"NitroDrome.Event.Reward.TimerStarted\")))"))
    (else
      (switch Utilities|FlowControl|Switch|SwitchonENitroDromeRewardStatus (CallFunction|GetCurrentRewardStatus)
        (:Active
          (Timer|PauseTimer))))))

(fn ConstructionScript)
