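; BP_NitroDrome_MapIconActor is an actor that carries a minimap icon for the NitroDrome mode.
; Its SetMapIcon function, on authority only, applies the given icon to the actor's Fort MiniMap component and flushes net dormancy, presumably so the change reaches clients.
; BeginPlay, overlap, tick and the construction script are empty.
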(event EventBeginPlay)

(event Collision|EventActorBeginOverlap
  (params (OtherActor :type Actor%20Object%20Reference)))

(event EventTick
  (params (DeltaSeconds :type Float%20%28single-precision%29)))

(fn SetMapIcon
  (params (MiniMapIcon :type Object%20Reference))
  (Networking|SwitchHasAuthority
    (:Authority
      (MiniMapData|SetMiniMapIcon (Variables|Default|GetFortMiniMapComponent) MiniMapIcon)
      (Networking|FlushNetDormancy))))

(fn ConstructionScript)
