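; NitroDrome gameplay volume actor that can locate the center loudspeaker actor (BP_NitroDrome_Center_A) overlapping it.
; GetLoudspeaker checks the stored Loudspeaker reference and, when it is invalid, casts and stores the first overlapping actor of that class.
; BeginPlay, overlap and tick events and the construction script are empty, so nothing here calls the lookup.
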(event EventBeginPlay)

(event Collision|EventActorBeginOverlap
  (params (OtherActor :type Actor%20Object%20Reference)))

(event EventTick
  (params (DeltaSeconds :type Float%20%28single-precision%29)))

(fn GetLoudspeaker
  (bind _overlappingactors (Collision|GetOverlappingActors "/NitroDrome/Assets/Blueprints/BP_NitroDrome_Center_A.BP_NitroDrome_Center_A_C"))
  (Utilities|IsValid (Variables|Default|GetLoudspeaker)
    (:Is%20Valid
      _)
    (:Is%20Not%20Valid
      (if (Utilities|Array|IsNotEmpty _overlappingactors)
        (bind _output_get (Variables|Default|SetLoudspeaker (Utilities|Casting|CastToBP_NitroDrome_Center_A (Utilities|Array|Get(acopy) _overlappingactors)))))))
  (return _output_get))

(fn ConstructionScript)
